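# Niagara player deploy effect for the Polished Jade glider, with a single "Fire" emitter that fires a delayed burst.
# Particles follow the skinned mesh's head bone in update, drifting outward on a lifetime-curved offset while stretching and fading.
# Layered mesh renderers get separate colors and spline-thickness material parameters, with jade or pink variant tints scaled by time of day.
# Spawn rate, spawn-time bone location and cone velocity are disabled.
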NiagaraSystem /BRCosmetics/Gadgets/Assets/VinderTech_GliderChute/Glider_PolishedJade/FX/NS_Glider_PolishedJade_PlayerDeploy
dependencies {
 usedDataInterfaces [&NiagaraDataInterfaceCurve, &NiagaraDataInterfaceSkeletalMesh]
 usedDynamicInputs [
  /BRCosmetics/Weapons/FORT_Melee/Pickaxe_GateHound/FX/VariantManager/VariantVectorSelector
  &FN_Select_Vector_from_Time_Of_Day
 ]
 usedModules [
  &SystemState
  &FN_VariantManager
  &self:SystemSpawnScript.SystemScriptSource.SystemScriptGraph.NiagaraNodeAssignment_0.SetVariables_#1
 ]
}
emitters {Fire {
  emitterUpdateScript [
   &EmitterState()
   &SpawnRate(
    SpawnRate=&Multiply_Float(
     A=10
     B=&FloatFromCurve(FloatCurve=DataInterface() {propertyValues {Curve {keys [<at 0 1>,<at 0.5 0>]}}}, CurveIndex=$Emitter.NormalizedLoopAge)
    )
    Use Spawn Probability=false
   ) {enabled=false}
   &SpawnBurst_Instantaneous(Spawn Time=0.1, Use Spawn Probability=false, UseLoopCountLimit=false)
   /SetVariables(
    Emitter.ErosionMax:NiagaraFloat=&FloatFromCurve(
     FloatCurve=DataInterface() {propertyValues {Curve {keys [
         {interpMode=RCIM_Cubic, tangentMode=RCTM_User, time=0, value=2}
         {arriveTangentWeight=0.0667, interpMode=RCIM_Cubic, leaveTangentWeight=0.1, tangentMode=RCTM_User, time=0.25, value=0.5}
         {arriveTangentWeight=0.2125, interpMode=RCIM_Cubic, leaveTangentWeight=0.3542, tangentMode=RCTM_User, time=0.65, value=0.5}
         {interpMode=RCIM_Cubic, tangentMode=RCTM_User, time=1, value=2}
        ]}}}
     CurveIndex=$Emitter.NormalizedLoopAge
    )
    Emitter.ErosionMin:NiagaraFloat=&FloatFromCurve(
     FloatCurve=DataInterface() {propertyValues {Curve {keys [
         <at 0 0.5 RCIM_Cubic>
         {arriveTangentWeight=0.0943, interpMode=RCIM_Cubic, leaveTangentWeight=0.1414, tangentMode=RCTM_User, time=0.25, value=0}
         {arriveTangentWeight=0.1179, interpMode=RCIM_Cubic, leaveTangentWeight=0.1965, tangentMode=RCTM_User, time=0.65, value=0}
         <at 1 0.5 RCIM_Cubic>
        ]}}}
     CurveIndex=$Emitter.NormalizedLoopAge
    )
   ) {isSetParameters=true}
  ]
  particleSpawnScript [
   &InitializeParticle(
    Color=&MultiplyLinearColor(
     Linear Color=&VariantColorSelector(Variant Amount=0, Original Color=<rgba 0.0018 0.0885 0.0857 0.5>, Variant 1 Color=<rgba 0.125 0.0182 0.0734 0.5>, Variant 2 Color=?)
     Scale Factor=&MakeFloat4FromVectorAndFloat("Vector (RGB)"=$System.ToD, "Float (Alpha)"=1)
    )
    Lifetime=&RandomRangeFloat(Minimum=1.6, Maximum=1.7)
    Mass Max=1.5
    Mass Mode=::NewEnumerator3
    Mesh Scale Mode=::NewEnumerator3
    Mesh Uniform Scale=10
    Sprite Size Mode=::NewEnumerator3
    Uniform Sprite Size=20
    UsePositionOffset=false
    Write Material Random=true
   )
   &SkeletalMeshLocation(
    Bone Sampling Mode=::NewEnumerator4
    MaskSpawnGroup=false
    Orientation Sampling=::NewEnumerator1
    Skeletal Mesh=DataInterface() {propertyValues {
      DefaultMesh=null
      ExcludeBoneName=root
      OptionalComponentClass=null
      PreviewMesh=/BRCosmetics/Characters/Player/Female/Medium/Bodies/F_MED_PolishedJade_Mind/Meshes/F_MED_PolishedJade_Mind
      SkinningMode=SkinOnTheFly
      SoftSourceActor=null
      SourceMode=Default
      UvSetIndex=0
      WholeMeshLOD=-1
      bExcludeBone=false
      bReadDeformedGeometry=true
      bRequireCurrentFrameData=true
      ComponentTags []
      FilteredBones [head]
      FilteredSockets []
      MeshUserParameter {parameter {name=None, typeDefHandle {registeredTypeIndex=1}, varData []}}
      SamplingRegions []
     }}
   ) {enabled=false, module=SkeletalMeshLocation001}
   &AddVelocityInCone(
    Cone Angle=30
    Cone Axis=&RotateVectorByQuaternion(Vector To Rotate=<0 1 0>, Quaternion=$Particles.SkeletalMeshLocation001.SampledRotation)
    Cone Axis Coordinate Space=::Simulation
    Module.Use Velocity Falloff On Cone Axis=false
    Velocity Strength=&RandomRangeFloat(Minimum=15, Maximum=100)
   ) {enabled=false}
   &InitialMeshOrientation(
    Mesh Orientation Mode=::NewEnumerator7
    Rotation=&Multiply_VectorByFloat(Vector=&RandomRangeVector(Minimum=<-1 0 -1>, Maximum=<1 0 1>), Float=0.02)
    Use Rotation Vector=true
   )
  ]
  particleUpdateScript [
   &ParticleState()
   &ScaleMeshSize(Scale Factor=&MakeVector(
     X=1
     Y=&FloatFromCurve(
      FloatCurve=DataInterface() {propertyValues {Curve {keys [
          {interpMode=RCIM_Cubic, tangentMode=RCTM_User, time=0, value=0.5}
          {arriveTangent=1.3454, interpMode=RCIM_Cubic, leaveTangent=1.3454, tangentMode=RCTM_User, time=0.5, value=1}
         ]}}}
      Scale Curve=&RandomRangeFloat(Minimum=0.8, Maximum=1)
     )
     Z=1
    ))
   &SkeletalMeshLocation(
    Bone Sampling Mode=::NewEnumerator4
    MaskSpawnGroup=false
    Sampled Position Offset=&MakeVector(
     X=0
     Y=&Lerp_Float(
      A=0
      B=&RandomRangeFloat(Minimum=15, Maximum=75)
      Alpha=&FloatFromCurve(FloatCurve=DataInterface() {propertyValues {Curve {keys [
           {arriveTangent=-0.1094, interpMode=RCIM_Cubic, leaveTangent=-0.1094, tangentMode=RCTM_User, time=0.15, value=0}
           {arriveTangent=3.0063, interpMode=RCIM_Cubic, leaveTangent=3.0063, tangentMode=RCTM_User, time=1, value=1}
          ]}}})
     )
     Z=0
    )
    Skeletal Mesh=DataInterface() {propertyValues {
      DefaultMesh=null
      ExcludeBoneName=root
      OptionalComponentClass=null
      PreviewMesh=&F_MED_PolishedJade_Mind
      SkinningMode=SkinOnTheFly
      SoftSourceActor=null
      SourceMode=Default
      UvSetIndex=0
      WholeMeshLOD=-1
      bExcludeBone=false
      bReadDeformedGeometry=true
      bRequireCurrentFrameData=true
      ComponentTags []
      FilteredBones [head]
      FilteredSockets []
      MeshUserParameter {parameter {name=None, typeDefHandle {registeredTypeIndex=1}, varData []}}
      SamplingRegions []
     }}
   )
   /SetVariables(Particles.AlphaScale:NiagaraFloat=&FloatFromCurve(
     FloatCurve=DataInterface() {propertyValues {Curve {keys [
         {arriveTangent=9.0725, interpMode=RCIM_Cubic, leaveTangent=9.0725, tangentMode=RCTM_User, time=0, value=0}
         {arriveTangent=0.1417, interpMode=RCIM_Cubic, leaveTangent=0.1417, tangentMode=RCTM_User, time=0.1, value=0.9847}
         {arriveTangent=-1.0941, interpMode=RCIM_Cubic, leaveTangent=-1.0941, time=0.8186, value=0.7135}
         {arriveTangent=-2.3882, interpMode=RCIM_Cubic, leaveTangent=-2.3882, tangentMode=RCTM_User, time=1, value=0}
        ]}}}
     Scale Curve=0.5
    )) {isSetParameters=true}
   /SetVariables(Particles.SplineThicknessMult=0.8) {isSetParameters=true}
   /SetVariables(Particles.URand:NiagaraFloat=&RandomRangeFloat(Minimum=0, Maximum=1)) {isSetParameters=true}
   &DynamicMaterialParameters(
    Index 0 Param 1=&Multiply_Float(A=65, B=$Particles.SplineThicknessMult)
    Index 0 Param 2=&Multiply_Float(A=14, B=$Particles.SplineThicknessMult)
    Index 0 Param 3=$Particles.URand
    Module.Index0Param0Write=true
    Module.Index0Param1Write=true
    Module.Index0Param2Write=true
    Module.Index0Param3Write=true
   )
   /SetVariables(
    Particles.Color2=&MultiplyLinearColor(
     Linear Color=&VariantColorSelector(Variant Amount=0, Original Color=<rgba 0.0048 0.6198 0.3447 1>, Variant 1 Color=<rgba 0.6042 0.2014 0.4296 1>, Variant 2 Color=?)
     Scale Factor=&MakeFloat4FromVectorAndFloat("Vector (RGB)"=$System.ToD, "Float (Alpha)"=$Particles.AlphaScale)
    )
    Particles.Dyn2:Vector4f=&MakeVector4(
     X=&Multiply_Float(A=50, B=$Particles.SplineThicknessMult)
     Y=&Multiply_Float(A=10, B=$Particles.SplineThicknessMult)
     Z=$Particles.URand
     W=0
    )
   ) {isSetParameters=true}
   /SetVariables(
    Particles.Color3:LinearColor=&MultiplyLinearColor(
     Linear Color=&VariantColorSelector(Variant Amount=0, Original Color=<rgba 0.1721 0.7031 0.6058 0.2>, Variant 1 Color=<rgba 0.6042 0.2014 0.4296 0.2>, Variant 2 Color=?)
     Scale Factor=&MakeFloat4FromVectorAndFloat("Vector (RGB)"=$System.ToD, "Float (Alpha)"=$Particles.AlphaScale)
    )
    Particles.Dyn3:Vector4f=&MakeVector4(
     X=&Multiply_Float(A=30, B=$Particles.SplineThicknessMult)
     Y=&Multiply_Float(A=8, B=$Particles.SplineThicknessMult)
     Z=$Particles.URand
     W=0
    )
   ) {isSetParameters=true}
   /SetVariables(
    Particles.Color4:LinearColor=&MultiplyLinearColor(
     Linear Color=&VariantColorSelector(Variant Amount=0, Original Color=<rgba 0.0048 0.6198 0.2303 0.25>, Variant 1 Color=<rgba 1.2 0.4937 0.9175 0.25>, Variant 2 Color=?)
     Scale Factor=&MakeFloat4FromVectorAndFloat("Vector (RGB)"=$System.ToD, "Float (Alpha)"=&Multiply_Float(A=$Particles.AlphaScale, B=0.5))
    )
    Particles.Dyn4:Vector4f=&MakeVector4(
     X=&Multiply_Float(A=50, B=$Particles.SplineThicknessMult)
     Y=&Multiply_Float(A=14, B=$Particles.SplineThicknessMult)
     Z=$Particles.URand
     W=0
    )
   ) {isSetParameters=true}
   &ScaleColor(Scale Alpha=$Particles.AlphaScale, ScaleA=true, ScaleRGB=true)
   &SolveForcesAndVelocity(Clamp Velocity=false, Limit Acceleration=false)
  ]
  rendererClasses [&NiagaraSpriteRendererProperties, &NiagaraMeshRendererProperties]
  renderers [
   {rendererClass=&NiagaraSpriteRendererProperties}
   {rendererClass=&NiagaraMeshRendererProperties, rendererIndex=1}
   {rendererClass=&NiagaraMeshRendererProperties, rendererIndex=2}
   {rendererClass=&NiagaraMeshRendererProperties, rendererIndex=3}
   {rendererClass=&NiagaraMeshRendererProperties, rendererIndex=4}
  ]
 }}
summary {
 systemName=NS_Glider_PolishedJade_PlayerDeploy
 emitters [{emitterName=Fire, rendererClasses [&NiagaraSpriteRendererProperties, &NiagaraMeshRendererProperties]}]
 userVariables [User.Variant {type=NiagaraFloat, defaultValue=0, description=""}]
}
