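# Character-parent material for the Polished Jade glider, adding an animated, flickering glow through Emissive Color between the shared PreFX and PostFX functions.
# A fire-flicker curve atlas, maxed with a slow sine, drives the flicker, with the mask's green channel blending parent and son speeds.
# The mask's red channel shapes that flicker, which gains a Fresnel rim and multiplies an emissive texture modulated by panning noise.
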Material /BRCosmetics/Gadgets/Assets/VinderTech_GliderChute/Glider_PolishedJade/Materials/M_Glider_PolishedJade
graph {
 properties {
  bUseMaterialAttributes=true
  bUsedWithMeshDeformer=true
  bUsedWithMorphTargets=true
  bUsedWithSkeletalMesh=true
  pixelDepthOffsetMode=PDOM_Legacy
  shadingModel=MSM_FromMaterialExpression
  usedShadingModels=Default Lit | Subsurface | Clear Coat | Subsurface Profile | Cloth
 }
 nodes [
  n1 = MaterialFunctionCall() {materialFunction=&MF_CharacterParent_2023_PreFX}
  MaterialFunctionCall(
   Material=SetMaterialAttributes(
    MaterialAttributes=GetMaterialAttributes(None=@n1.Result) {attributeGetTypes [#1]}.MaterialAttributes
    Emissive Color=NamedRerouteUsage()
   ) {attributeSetTypes [#1]}
   M=@n1.M
   Pre Skinned Local Position=@n1.M
   Distance Blend=@n1.M
   Fresnel=@n1.M
  ) {materialFunction=&MF_CharacterParent_2023_PostFX}
  n2 = Time()
  Multiply(A=@n2) {constB=0.5}
  n3 = Sine(None=Multiply(A=Time()) {constB=0.5})
  n4 = TextureSample() {texture=/BRCosmetics/Gadgets/Assets/VinderTech_GliderChute/Glider_PolishedJade/Textures/PolishedJade_Glider_Mask_E}
  n5 = LinearInterpolate(
   A=LinearInterpolate(
    A=ScalarParameter() {defaultValue=0.8, group=Glider_PolishedJade, parameterName=FlickerSpeed_parent}
    B=ScalarParameter() {defaultValue=0.4, group=Glider_PolishedJade, parameterName=FlickerSpeed_son}
    Alpha=@n4.G
   ) {constB=0.25}
   B=ScalarParameter() {defaultValue=0.5, group=Glider_PolishedJade, parameterName=FlickerSpeed2}
   Alpha=Frac(None=ScalarParameter() {defaultValue=1, group=Glider_PolishedJade, parameterName=RandomLightVal})
  )
  n6 = CurveAtlasRowParameter(CurveTime=Frac(None=Multiply(A=@n2, B=@n5))) {
   atlas=/Game/Environments/Helios/Props/Brimstone/Brimstone_Brazier_A/Blueprints/CurveAtlas_FN_LightFunctions
   curve=/Game/Blueprints/Lighting/LightFunctions/Curve_FireFlickers
   parameterName=FireFlicker
  }
  n7 = Saturate(None=@n6.R)
  NamedRerouteDeclaration(None=LinearInterpolate(A=ScalarParameter() {defaultValue=0.6, group=Glider_PolishedJade, parameterName=Subtle lighting}, Alpha=Max(A=@n3, B=@n7)) {constA=0.85}) {name=FlickerLight, nodeColor=<rgba 0.0078 1 0 1>}
  n8 = TextureSampleParameter2D() {
   group=Glider_PolishedJade
   parameterName=EmissiveColor
   texture=/BRCosmetics/Gadgets/Assets/VinderTech_GliderChute/Glider_PolishedJade/Textures/T_Glider_PolishedJade_E
  }
  n9 = TextureCoordinate()
  n10 = TextureSample(UVs=Panner(Coordinate=Multiply(A=@n9) {constB=2}) {speedY=0.05}) {samplerSource=SSM_Wrap_WorldGroupSettings, samplerType=SAMPLERTYPE_Grayscale, texture=&LowResBlurredNoise}
  n11 = Power(Base=@n10.R)
  n12 = Multiply(
   A=@n8.RGB
   B=Multiply(
    A=LinearInterpolate(Alpha=@n11) {constA=0.3}
    B=ScalarParameter() {defaultValue=8, group=Glider_PolishedJade, parameterName=Emissive intensity}
   )
  ) {constB=4}
  n13 = Multiply(A=@n4.R, B=Multiply(A=Constant() {r=1.2}, B=NamedRerouteUsage())) {constB=2}
  n14 = NamedRerouteUsage()
  n15 = LinearInterpolate(A=Constant() {r=2}, B=Constant() {r=1}, Alpha=Saturate(None=Multiply(A=@n4.R, B=@n14))) {constA=1}
  n16 = Saturate(None=Power(Base=Fresnel()) {constExponent=1})
  n17 = Add(A=Saturate(None=Power(Base=@n13, Exp=@n15)), B=Multiply(A=Multiply(A=@n8.B, B=@n16)) {constB=0.25})
  NamedRerouteDeclaration(None=MaterialFunctionCall(Emissive=Multiply(A=@n12, B=@n17)) {materialFunction=&MF_FX_EmissiveAdjustments}.Result) {name=out_Emissive, nodeColor=<rgba 0.9922 1 0 1>}
 ]
}
