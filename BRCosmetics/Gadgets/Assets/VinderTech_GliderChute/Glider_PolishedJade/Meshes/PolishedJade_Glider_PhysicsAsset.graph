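# Physics asset for the PolishedJade glider cosmetic, a fish-shaped rig with capsule bodies on its dyn_ bones.
# Constraints with uniform 45-degree swing and twist limit values chain the big fish's head, mouth, whiskers, spine, fins and tail, plus a platform off the head.
# The left small fish's spine and top fin form a separate chain, and its head body has no constraint.
# Right front fin capsules are far larger than the left, with an extra segment.
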PhysicsAsset /BRCosmetics/Gadgets/Assets/VinderTech_GliderChute/Glider_PolishedJade/Meshes/PolishedJade_Glider_PhysicsAsset
bodies {
 dyn_big_fish_head [{shapeType=Capsule, length=90.5204, radius=86.8378, center=<-11.0198 0.1925 10.4501>, rotation=<rot 0.125 90.6327 -90>}]
 dyn_big_fish_neck [{shapeType=Capsule, length=0.505, radius=0.505, rotation=<rot 0 -90 0>}]
 dyn_fin_l_bk_01 [{shapeType=Capsule, length=20.4592, radius=11.3316, center=<25.9711 -0.1553 3.4278>, rotation=<rot 0 -90.4739 -69.6402>}]
 dyn_fin_l_fr_02 [{shapeType=Capsule, length=18.8671, radius=11.0571, center=<10.6667 1.705 1.2566>, rotation=<rot 8.1897 -90 -0.599>}]
 dyn_fin_l_fr_03 [{shapeType=Capsule, length=21.7432, radius=18.8451, center=<7.578 2.0802 4.4329>, rotation=<rot 7.465 -90 -1.4004>}]
 dyn_fin_l_fr_04 [{shapeType=Capsule, length=23.2262, radius=16.4964, center=<18.1905 1.5419 0.9336>, rotation=<rot 0 -92.5517 -72.3637>}]
 dyn_fin_r_fbk_01 [{shapeType=Capsule, length=20.4592, radius=11.3316, center=<-25.9268 0.0364 -3.4266>, rotation=<rot 0 -90.4739 -69.6402>}]
 dyn_fin_r_fr_02 [{shapeType=Capsule, length=62.4906, radius=20.8426, center=<-17.1695 52.673 -17.0473>, rotation=<rot 0 7.1643 -75.3254>}]
 dyn_fin_r_fr_03 [{shapeType=Capsule, length=72.567, radius=33.5716, center=<-7.2029 68.0216 -7.1744>, rotation=<rot 0 -0.5821 -80.6164>}]
 dyn_fin_r_fr_04 [{shapeType=Capsule, length=78.4829, radius=65.0218, center=<-24.098 68.4979 8.3573>, rotation=<rot 0 -37.7368 -76.8999>}]
 dyn_fin_r_fr_05 [{shapeType=Capsule, length=66.4244, radius=51.0124, center=<-35.1008 70.9622 8.0114>, rotation=<rot 0 -14.7395 -79.7668>}]
 dyn_fin_top_03 [{shapeType=Capsule, length=16.2735, radius=13.878, center=<14.9217 -3.0439 0.1841>, rotation=<rot 0 -111.4504 -89.5426>}]
 dyn_mouth_lower [{shapeType=Capsule, length=26.7776, radius=7.8097, center=<10.6457 2.6892 0>, rotation=<rot 0 -90 0>}]
 dyn_mouth_upper [{shapeType=Capsule, length=30.6852, radius=12.0253, center=<8.0413 -2.2051 0>, rotation=<rot 0 -90 0>}]
 dyn_platform_01 [{shapeType=Capsule, length=39.7032, radius=20.1557, center=<32.3051 -0.3147 -5.6635>, rotation=<rot 0 -1.4392 -89.2158>}]
 dyn_platform_03 [{shapeType=Capsule, length=35.0148, radius=31.7078, center=<3.7496 -0.0384 -0.6818>, rotation=<rot 0 93.0242 -87.2202>}]
 dyn_platform_base [{shapeType=Capsule, length=45.5487, radius=13.8222, center=<0.7624 9.3776 0.4266>, rotation=<rot 2.1352 -90 -0.4146>}]
 dyn_sm_l_spine_01 [{shapeType=Capsule, length=39.6402, radius=36.7841, center=<7.1674 -0.3519 -1.964>, rotation=<rot 0 179.8941 -89.3059>}]
 dyn_sm_l_spine_02 [{shapeType=Capsule, length=32.3489, radius=28.0197, center=<16.234 0.0624 -0.5864>, rotation=<rot 6.3947 -90 -0.0451>}]
 dyn_sm_l_spine_04 [{shapeType=Capsule, length=24.5108, radius=21.0366, center=<5.6835 -0.0097 -0.8541>, rotation=<rot 5.5 -90 -0.2919>}]
 dyn_sm_l_spine_06 [{shapeType=Capsule, length=29.7323, radius=18.3028, center=<18.0227 0.0112 5.6785>, rotation=<rot 0.2137 -90 8.7112>}]
 dyn_small_fish_head_l [{shapeType=Capsule, length=37.5866, radius=35.5804, center=<14.9844 0.134 0.4607>, rotation=<rot 0 -0.0856 -89.8223>}]
 dyn_smleft_fin_top_01 [{shapeType=Capsule, length=23.7308, radius=18.1772, center=<9.5734 0.015 2.246>, rotation=<rot 0 90 -51.2638>}]
 dyn_spine_01 [{shapeType=Capsule, length=63.7083, radius=15.2907, center=<25.2673 0.0634 -44.6895>, rotation=<rot 0 180 -90>}]
 dyn_spine_05 [{shapeType=Capsule, length=61.2055, radius=57.6114, center=<4.6431 6.1755 -1.8556>, rotation=<rot 0 -169.6793 -154.7111>}]
 dyn_spine_07 [{shapeType=Capsule, length=42.0216, radius=30.6632, center=<4.3115 0.0333 4.1103>, rotation=<rot -0.4742 -90 -10.5598>}]
 dyn_tail_base [{shapeType=Capsule, length=35.7615, radius=30.6191, center=<22.7117 1.5395 3.5981>, rotation=<rot 0.8188 -90 -24.8924>}]
 dyn_tail_bot_02 [{shapeType=Capsule, length=25.9325, radius=9.3309, center=<5.2601 0.0109 9.7592>, rotation=<rot 0.2955 -90 -5.1361>}]
 dyn_tail_bot_03 [{shapeType=Capsule, length=17.5479, radius=10.2672, center=<10.7903 0 -0.6444>, rotation=<rot 0 -90 3.7836>}]
 dyn_tail_bot_04 [{shapeType=Capsule, length=33.4135, radius=14.1527, center=<27.4142 -0.1887 5.6804>, rotation=<rot 0 89.5616 -63.1423>}]
 dyn_tail_bot_05 [{shapeType=Capsule, length=27.5642, radius=10.7804, center=<17.4395 0.0284 10.2855>, rotation=<rot 0 90.1013 -49.6007>}]
 dyn_tail_top_01 [{shapeType=Capsule, length=21.1107, radius=6.9708, center=<8.8215 -0.7222 -14.0846>, rotation=<rot 4.1973 -90 -1.8799>}]
 dyn_tail_top_02 [{shapeType=Capsule, length=22.9952, radius=12.7126, center=<11.7489 0 -5.1665>, rotation=<rot 0 -90 -7.1506>}]
 dyn_tail_top_03 [{shapeType=Capsule, length=23.5477, radius=10.8443, center=<12.4633 0 -4.9826>, rotation=<rot 0 -90 -7.2957>}]
 dyn_tail_top_04 [{shapeType=Capsule, length=20.0788, radius=9.1953, center=<9.8838 0 -4.5976>, rotation=<rot 0 -90 3.9463>}]
 dyn_tail_top_05 [{shapeType=Capsule, length=28.6688, radius=9.8653, center=<28.8084 0 -8.3719>, rotation=<rot 0 -90 -69.5597>}]
 dyn_whisker_l_01 [{shapeType=Capsule, length=18.5367, radius=5.3429, center=<16.3113 -0.1671 0.0952>, rotation=<rot 0 -89.1076 -89.5432>}]
 dyn_whisker_l_03 [{shapeType=Capsule, length=21.2741, radius=10.5971, center=<21.5401 0.2544 -0.1293>, rotation=<rot 0 -89.7478 -89.3059>}]
 dyn_whisker_l_05 [{shapeType=Capsule, length=23.0193, radius=13.8409, center=<20.6095 0.5935 0.7306>, rotation=<rot 0 92.041 -75.3549>}]
 dyn_whisker_r_01 [{shapeType=Capsule, length=18.5367, radius=5.3429, center=<-16.2622 0.0505 -0.0846>, rotation=<rot 0 -89.1076 -89.5432>}]
 dyn_whisker_r_03 [{shapeType=Capsule, length=21.2741, radius=10.5971, center=<-21.4906 -0.3707 0.1405>, rotation=<rot 0 -89.7478 -89.3059>}]
 dyn_whisker_r_05 [{shapeType=Capsule, length=23.0193, radius=13.8409, center=<-20.5608 -0.7101 -0.7192>, rotation=<rot 0 92.041 -75.3548>}]
}
constraints [
 {bone1Name=dyn_big_fish_head, bone2Name=dyn_big_fish_neck, swing1LimitDegrees=45, swing2LimitDegrees=45, twistLimitDegrees=45}
 {bone1Name=dyn_platform_base, bone2Name=dyn_big_fish_head, swing1LimitDegrees=45, swing2LimitDegrees=45, twistLimitDegrees=45}
 {bone1Name=dyn_platform_01, bone2Name=dyn_platform_base, swing1LimitDegrees=45, swing2LimitDegrees=45, twistLimitDegrees=45}
 {bone1Name=dyn_platform_03, bone2Name=dyn_platform_01, swing1LimitDegrees=45, swing2LimitDegrees=45, twistLimitDegrees=45}
 {bone1Name=dyn_fin_top_03, bone2Name=dyn_big_fish_head, swing1LimitDegrees=45, swing2LimitDegrees=45, twistLimitDegrees=45}
 {bone1Name=dyn_mouth_lower, bone2Name=dyn_big_fish_head, swing1LimitDegrees=45, swing2LimitDegrees=45, twistLimitDegrees=45}
 {bone1Name=dyn_mouth_upper, bone2Name=dyn_big_fish_head, swing1LimitDegrees=45, swing2LimitDegrees=45, twistLimitDegrees=45}
 {bone1Name=dyn_whisker_l_01, bone2Name=dyn_big_fish_head, swing1LimitDegrees=45, swing2LimitDegrees=45, twistLimitDegrees=45}
 {bone1Name=dyn_whisker_l_03, bone2Name=dyn_whisker_l_01, swing1LimitDegrees=45, swing2LimitDegrees=45, twistLimitDegrees=45}
 {bone1Name=dyn_whisker_l_05, bone2Name=dyn_whisker_l_03, swing1LimitDegrees=45, swing2LimitDegrees=45, twistLimitDegrees=45}
 {bone1Name=dyn_whisker_r_01, bone2Name=dyn_big_fish_head, swing1LimitDegrees=45, swing2LimitDegrees=45, twistLimitDegrees=45}
 {bone1Name=dyn_whisker_r_03, bone2Name=dyn_whisker_r_01, swing1LimitDegrees=45, swing2LimitDegrees=45, twistLimitDegrees=45}
 {bone1Name=dyn_whisker_r_05, bone2Name=dyn_whisker_r_03, swing1LimitDegrees=45, swing2LimitDegrees=45, twistLimitDegrees=45}
 {bone1Name=dyn_spine_01, bone2Name=dyn_big_fish_neck, swing1LimitDegrees=45, swing2LimitDegrees=45, twistLimitDegrees=45}
 {bone1Name=dyn_spine_05, bone2Name=dyn_spine_01, swing1LimitDegrees=45, swing2LimitDegrees=45, twistLimitDegrees=45}
 {bone1Name=dyn_spine_07, bone2Name=dyn_spine_05, swing1LimitDegrees=45, swing2LimitDegrees=45, twistLimitDegrees=45}
 {bone1Name=dyn_tail_base, bone2Name=dyn_spine_07, swing1LimitDegrees=45, swing2LimitDegrees=45, twistLimitDegrees=45}
 {bone1Name=dyn_tail_top_01, bone2Name=dyn_tail_base, swing1LimitDegrees=45, swing2LimitDegrees=45, twistLimitDegrees=45}
 {bone1Name=dyn_tail_top_02, bone2Name=dyn_tail_top_01, swing1LimitDegrees=45, swing2LimitDegrees=45, twistLimitDegrees=45}
 {bone1Name=dyn_tail_top_03, bone2Name=dyn_tail_top_02, swing1LimitDegrees=45, swing2LimitDegrees=45, twistLimitDegrees=45}
 {bone1Name=dyn_tail_top_04, bone2Name=dyn_tail_top_03, swing1LimitDegrees=45, swing2LimitDegrees=45, twistLimitDegrees=45}
 {bone1Name=dyn_tail_top_05, bone2Name=dyn_tail_top_04, swing1LimitDegrees=45, swing2LimitDegrees=45, twistLimitDegrees=45}
 {bone1Name=dyn_tail_bot_02, bone2Name=dyn_tail_base, swing1LimitDegrees=45, swing2LimitDegrees=45, twistLimitDegrees=45}
 {bone1Name=dyn_tail_bot_03, bone2Name=dyn_tail_bot_02, swing1LimitDegrees=45, swing2LimitDegrees=45, twistLimitDegrees=45}
 {bone1Name=dyn_tail_bot_04, bone2Name=dyn_tail_bot_03, swing1LimitDegrees=45, swing2LimitDegrees=45, twistLimitDegrees=45}
 {bone1Name=dyn_tail_bot_05, bone2Name=dyn_tail_bot_04, swing1LimitDegrees=45, swing2LimitDegrees=45, twistLimitDegrees=45}
 {bone1Name=dyn_fin_l_bk_01, bone2Name=dyn_spine_05, swing1LimitDegrees=45, swing2LimitDegrees=45, twistLimitDegrees=45}
 {bone1Name=dyn_fin_r_fbk_01, bone2Name=dyn_spine_05, swing1LimitDegrees=45, swing2LimitDegrees=45, twistLimitDegrees=45}
 {bone1Name=dyn_fin_l_fr_02, bone2Name=dyn_spine_01, swing1LimitDegrees=45, swing2LimitDegrees=45, twistLimitDegrees=45}
 {bone1Name=dyn_fin_l_fr_03, bone2Name=dyn_fin_l_fr_02, swing1LimitDegrees=45, swing2LimitDegrees=45, twistLimitDegrees=45}
 {bone1Name=dyn_fin_l_fr_04, bone2Name=dyn_fin_l_fr_03, swing1LimitDegrees=45, swing2LimitDegrees=45, twistLimitDegrees=45}
 {bone1Name=dyn_fin_r_fr_02, bone2Name=dyn_spine_01, swing1LimitDegrees=45, swing2LimitDegrees=45, twistLimitDegrees=45}
 {bone1Name=dyn_fin_r_fr_03, bone2Name=dyn_fin_r_fr_02, swing1LimitDegrees=45, swing2LimitDegrees=45, twistLimitDegrees=45}
 {bone1Name=dyn_fin_r_fr_04, bone2Name=dyn_fin_r_fr_03, swing1LimitDegrees=45, swing2LimitDegrees=45, twistLimitDegrees=45}
 {bone1Name=dyn_fin_r_fr_05, bone2Name=dyn_fin_r_fr_04, swing1LimitDegrees=45, swing2LimitDegrees=45, twistLimitDegrees=45}
 {bone1Name=dyn_sm_l_spine_02, bone2Name=dyn_sm_l_spine_01, swing1LimitDegrees=45, swing2LimitDegrees=45, twistLimitDegrees=45}
 {bone1Name=dyn_sm_l_spine_04, bone2Name=dyn_sm_l_spine_02, swing1LimitDegrees=45, swing2LimitDegrees=45, twistLimitDegrees=45}
 {bone1Name=dyn_sm_l_spine_06, bone2Name=dyn_sm_l_spine_04, swing1LimitDegrees=45, swing2LimitDegrees=45, twistLimitDegrees=45}
 {bone1Name=dyn_smleft_fin_top_01, bone2Name=dyn_sm_l_spine_02, swing1LimitDegrees=45, swing2LimitDegrees=45, twistLimitDegrees=45}
]
